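# Physics asset for the Genius pickaxe cosmetic melee weapon's mesh, simulating a dynamic bubble piece on the axe.
# It holds two bodies: a small sphere on the dyn_bubble_MSTR bone and a short capsule on dyn_bubble.
# A single constraint joins them, allowing wide swing on both axes but little twist, so the bubble can sway relative to its master bone.
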PhysicsAsset /BRCosmetics/Weapons/FORT_Melee/Pickaxe_Genius/Meshes/Genius_Axe_Physics
bodies {
 dyn_bubble [{shapeType=Capsule, length=3, radius=1.8, shapeName=dyn_bubble_capsule}]
 dyn_bubble_MSTR [{shapeType=Sphere, radius=2, shapeName=dyn_bubble_MSTR_sphere}]
}
constraints [{bone1Name=dyn_bubble_MSTR, bone2Name=dyn_bubble, swing1LimitDegrees=80, swing2LimitDegrees=80, twistLimitDegrees=10}]
