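# Animation Blueprint for the Genius pickaxe that adds secondary physics motion to its dyn_bubble bone.
# One spring-based Anim Dynamics node drives the output, with alpha and gravity override fed by variables.
# The update event reads the owning pawn's bWasJumping, snapping a blend to jump values (lower alpha, stronger downward gravity) and easing back afterward.
# Alternative Anim Dynamics, RigidBody and Spring controller nodes are left unconnected to the output.
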AnimBlueprint /BRCosmetics/Weapons/FORT_Melee/Pickaxe_Genius/Meshes/Genius_Axe_AnimBP
target_skeleton=&/Genius_Axe_Skeleton
graphs [
 AnimGraph {nodes [
   Anim Dynamics - Bone: dyn_bubble!(ComponentPose, bAlphaBoolEnabled:bool=True, GravityOverride:Vector="0,0,-430", ExternalForce="0,0,0") -> (Pose) {
    alphaInputType=Bool
    angularBiasOverride=0.98
    angularDampingOverride=0.98
    angularSpringConstant=230
    bAngularSpring=true
    bLinearSpring=true
    bOverrideAngularBias=true
    bOverrideAngularDamping=true
    bOverrideLinearDamping=true
    bUseGravityOverride=true
    linearDampingOverride=0.85
    linearSpringConstant=230
    alphaBoolBlend {blendOption=Linear}
    boundBone {boneName=dyn_bubble}
    componentAppliedLinearAccClamp=<300 300 100>
    componentLinearAccScale=<0.2 0.2 0.01>
    componentLinearVelScale=<0.05 0.05 0.01>
    physicsBodyDefinitions [{
      boundBone {boneName=dyn_bubble}
      boxExtents=<110 110 110>
      constraintSetup {
       angularLimitsMax=<20 20 20>
       angularLimitsMin=<-20 -20 -20>
       angularTarget=<1 0 0>
       linearAxesMax=<0.4 0.6 0.4>
       linearAxesMin=<-0.4 -0.6 -0.4>
      }
      localJointOffset=<0 0 0>
     }]
   }
   Anim Dynamics - Bone: dyn_bubble(bAlphaBoolEnabled=True, GravityOverride="0,0,-100", ExternalForce="0,0,0") {
    alphaInputType=Bool
    angularBiasOverride=0.98
    angularDampingOverride=0.98
    angularSpringConstant=230
    bAngularSpring=true
    bLinearSpring=true
    bOverrideAngularBias=true
    bOverrideAngularDamping=true
    bOverrideLinearDamping=true
    bUseGravityOverride=true
    linearDampingOverride=0.85
    linearSpringConstant=230
    alphaBoolBlend {blendOption=Linear}
    boundBone {boneName=dyn_bubble}
    componentAppliedLinearAccClamp=<300 300 100>
    componentLinearAccScale=<0.2 0.2 0.01>
    componentLinearVelScale=<0.05 0.05 0.01>
    physicsBodyDefinitions [{
      boundBone {boneName=dyn_bubble}
      boxExtents=<110 110 110>
      constraintSetup {
       angularLimitsMax=<20 20 20>
       angularLimitsMin=<-20 -20 -20>
       angularTarget=<1 0 0>
       linearAxesMax=<0.2 0.6 0.4>
       linearAxesMin=<-0.4 -0.6 -0.4>
      }
      localJointOffset=<0 0 0>
     }]
   }
   RigidBody!(ComponentPose, Alpha, OverrideWorldGravity:Vector="0,0,0", ExternalForce) -> (Pose) {
    bClampLinearTranslationLimitToRefPose=true
    bOverrideWorldGravity=true
    overridePhysicsAsset=&/Genius_Axe_Physics
    alphaBoolBlend {blendOption=Linear}
    componentAppliedLinearAccClamp=<5000 5000 5000>
    componentLinearAccScale=<0.4 0.4 0.4>
    componentLinearVelScale=<0.4 0.4 0.4>
   }
   AnimGraphNode_SubInput_0 = Input Pose()
   n1 = Anim Dynamics - Bone: dyn_bubble!(
    ComponentPose=Local To Component(LocalPose=@AnimGraphNode_SubInput_0)
    Alpha:float=Get AnimDynamicAlpha() -> (AnimDynamicAlpha:double)
    GravityOverride=Get Gravity_override() -> (Gravity_override:Vector)
    ExternalForce="0,0,0"
   ) -> (Pose) {
    angularBiasOverride=0.98
    angularDampingOverride=0.98
    angularSpringConstant=230
    bAngularSpring=true
    bLinearSpring=true
    bOverrideAngularBias=true
    bOverrideAngularDamping=true
    bOverrideLinearDamping=true
    bUseGravityOverride=true
    linearDampingOverride=0.9
    linearSpringConstant=240
    alphaBoolBlend {blendOption=Linear}
    boundBone {boneName=dyn_bubble}
    componentAppliedLinearAccClamp=<300 300 100>
    componentLinearAccScale=<0.2 0.2 0.01>
    componentLinearVelScale=<0.05 0.05 0.01>
    physicsBodyDefinitions [{
      boundBone {boneName=dyn_bubble}
      boxExtents=<110 110 110>
      constraintSetup {
       angularLimitsMax=<20 20 20>
       angularLimitsMin=<-20 -20 -20>
       angularTarget=<1 0 0>
       linearAxesMax=<0.2 0.6 0.4>
       linearAxesMin=<-0.4 -0.6 -0.4>
      }
      localJointOffset=<0 0 0>
     }]
   }
   Output Pose(Result=Component To Local(ComponentPose=@n1)) {layerGroup=None}
   Spring controller - Bone: dyn_bubble!(ComponentPose:Component Space Pose Link Structure, Alpha:float=1, SpringStiffness:double=800, SpringDamping:double=800) -> (Pose:Component Space Pose Link Structure) {
    is=title AnimGraphNode_SpringBone
    bLimitDisplacement=true
    bRotateX=true
    bRotateY=true
    bRotateZ=true
    maxDisplacement=1.5
    springDamping=10
    springStiffness=10
    alphaBoolBlend {blendOption=Linear}
    springBone {boneName=dyn_bubble}
   }
   NOT Boolean(A=Get bRigidBodyEnabled())
  ]}
 EventGraph {nodes [
   n1 = Event BlueprintUpdateAnimation()
   > Is Valid(InputObject=Get As Player Pawn Athena() -> (As Player Pawn Athena:Player Pawn Athena Object Reference))
   n2 = Get bWasJumping!(self:Character Object Reference=Get As Player Pawn Athena()) -> (bWasJumping:bool)
   > n3 = Branch(Condition=@n2)
   > n4 = Set InterpolateAlpha(InterpolateAlpha:double=1) -> (Output_Get:double)
   n5 = Set InterpolateAlpha(
    execute=[@n4.then, @n3.else]
    InterpolateAlpha=FInterpTo(
     Current=Get InterpolateAlpha() -> (InterpolateAlpha:double)
     Target="To Float (Boolean)"(InBool=@n2)
     DeltaTime=@n1.DeltaTimeX
     InterpSpeed=3
    )
   )
   > Set AnimDynamicAlpha(AnimDynamicAlpha:double=Lerp(A=1, B=0.65, Alpha=@n5)) -> (Output_Get:double)
   > Set Gravity_override(Gravity_override:Vector="Lerp (Vector)"!(A:Vector="0, 0,-100", B:Vector="0, 0,-600", Alpha:float=@n5) -> (ReturnValue:Vector)) -> (Output_Get:Vector)
   Event BlueprintInitializeAnimation()
   > n6 = Cast To PlayerPawn_Athena(Object=TryGetPawnOwner())
   > Set As Player Pawn Athena(As Player Pawn Athena:Player Pawn Athena Object Reference=@n6) -> (Output_Get:Player Pawn Athena Object Reference)
   Get bIsJumping()
  ]}
]
variables [
 AnimDynamicAlpha {type=number}
 As Player Pawn Athena {type=PlayerPawn_Athena_C}
 Gravity_override {type=Vector}
 InterpolateAlpha {type=number}
]
